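# Physics asset for the male medium TechSpeeder cosmetic body, set up for secondary motion of the jacket's dyn_jacket bones.
# Capsules cover pelvis, spine_03, thighs, upper arms and the front, side and back jacket pieces.
# Small anchor capsules link to spine_03 with free angular motion, and each drives an angle-limited flap capsule.
# Extra limited constraints tie upper front flaps to mid front flaps on each side, coupling their swing.
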PhysicsAsset /BRCosmetics/Characters/Player/Male/Medium/Bodies/M_MED_TechSpeeder/Meshes/M_MED_TechSpeeder_Physics
bodies {
 dyn_jacket_down_back_01 [{
   shapeType=Capsule
   length=9.8
   radius=2.4
   shapeName=dyn_jacket_down_front_01_l_capsule
   center=<4.0483 0 1.5138>
   rotation=<rot 90 153.4349 -116.5651>
  }]
 dyn_jacket_down_back_base [{shapeType=Capsule, length=0.505, radius=0.505, shapeName=dyn_jacket_down_back_base_capsule, rotation=<rot 0 -90 0>}]
 dyn_jacket_down_front_01_l [{
   shapeType=Capsule
   length=7
   radius=4
   shapeName=dyn_jacket_down_front_01_l_capsule
   center=<-4.9829 1.486 0.1903>
   rotation=<rot 75.9152 -81.6645 -90.0001>
  }]
 dyn_jacket_down_front_01_r [{
   shapeType=Capsule
   length=7
   radius=4
   shapeName=dyn_jacket_down_front_01_r_capsule
   center=<4.9829 -1.486 -0.1903>
   rotation=<rot 90 0 171.6643>
  }]
 dyn_jacket_down_front_base_l [{shapeType=Capsule, length=0.1, radius=0.1, shapeName=dyn_jacket_down_front_base_l_capsule, rotation=<rot 0 -90 0>}]
 dyn_jacket_down_front_base_r [{shapeType=Capsule, length=0.1, radius=0.1, shapeName=dyn_jacket_down_front_base_r_capsule, rotation=<rot 0 -90 0>}]
 dyn_jacket_down_side_01_l [{shapeType=Capsule, length=0.505, radius=0.505, shapeName=dyn_jacket_down_side_01_l_capsule, rotation=<rot 0 -90 0>}]
 dyn_jacket_down_side_01_r [{shapeType=Capsule, length=0.505, radius=0.505, shapeName=dyn_jacket_down_side_01_r_capsule, rotation=<rot 0 -90 0>}]
 dyn_jacket_down_side_base_l [{shapeType=Capsule, length=3.6, radius=4.2, shapeName=dyn_jacket_down_front_01_l_capsule, center=<-4.8727 0 0>, rotation=<rot 90 0 0>}]
 dyn_jacket_down_side_base_r [{shapeType=Capsule, length=3.6, radius=4.2, shapeName=dyn_jacket_down_front_01_r_capsule, center=<4.8727 0 0>, rotation=<rot 90 0 180>}]
 dyn_jacket_mid_front_01_l [{
   shapeType=Capsule
   length=20.4
   radius=3.1
   shapeName=dyn_jacket_up_front_01_l_capsule
   center=<0.0881 3.6103 0>
   rotation=<rot 86.7516 179.9998 0>
  }]
 dyn_jacket_mid_front_01_r [{
   shapeType=Capsule
   length=20.4
   radius=3.1
   shapeName=dyn_jacket_up_front_01_r_capsule
   center=<-0.0881 -3.6103 0>
   rotation=<rot 86.7516 179.9998 0.0001>
  }]
 dyn_jacket_mid_front_base_l [{shapeType=Capsule, length=0.1, radius=0.1, shapeName=dyn_jacket_mid_front_base_l_capsule, rotation=<rot 0 -90 0>}]
 dyn_jacket_mid_front_base_r [{shapeType=Capsule, length=0.1, radius=0.1, shapeName=dyn_jacket_mid_front_base_r_capsule, rotation=<rot 0 -90 0>}]
 dyn_jacket_up_front_01_l [{shapeType=Capsule, length=10.9, radius=3.4, shapeName=dyn_jacket_up_front_01_l_capsule, center=<5.7356 3.6644 0>, rotation=<rot 90 0 0>}]
 dyn_jacket_up_front_01_r [{
   shapeType=Capsule
   length=10.9
   radius=3.4
   shapeName=dyn_jacket_up_front_01_r_capsule
   center=<-5.7356 -3.6644 0.1383>
   rotation=<rot 90 0 180>
  }]
 dyn_jacket_up_front_base_l [{shapeType=Capsule, length=0.1, radius=0.1, shapeName=dyn_jacket_up_front_base_l_capsule, rotation=<rot 0 -90 0>}]
 dyn_jacket_up_front_base_r [{shapeType=Capsule, length=0.1, radius=0.1, shapeName=dyn_jacket_up_front_base_r_capsule, rotation=<rot 0 -90 0>}]
 pelvis [{shapeType=Capsule, length=10.3495, radius=11.2762, shapeName=pelvis_capsule, rotation=<rot 0 -90 0>}]
 spine_03 [{shapeType=Capsule, length=7.7926, radius=8.1, shapeName=spine_03_capsule, rotation=<rot 0 -90 90>}]
 thigh_l [{shapeType=Capsule, length=20.2, radius=9.8, shapeName=thigh_l_capsule, center=<-8.9375 0.4157 0>, rotation=<rot 0 -90 -90>}]
 thigh_r [{shapeType=Capsule, length=20.2, radius=9.8, shapeName=thigh_r_capsule, center=<8.9375 -0.4157 0>, rotation=<rot 0 -90 90>}]
 upperarm_l [{shapeType=Capsule, length=26.9, radius=3.5, shapeName=upperarm_l_capsule, center=<14.688 0.167 -0.763>, rotation=<rot 0 90 90>}]
 upperarm_r [{shapeType=Capsule, length=25.3, radius=5.1, shapeName=upperarm_r_capsule, center=<-14.688 -0.167 -0.7392>, rotation=<rot 0 90 -90>}]
}
constraints [
 {bone1Name=dyn_jacket_up_front_01_r, bone2Name=dyn_jacket_up_front_base_r, swing1LimitDegrees=2, swing2LimitDegrees=6, twistLimitDegrees=45}
 {bone1Name=dyn_jacket_up_front_01_l, bone2Name=dyn_jacket_up_front_base_l, swing1LimitDegrees=2, swing2LimitDegrees=6, twistLimitDegrees=45}
 {
  bone1Name=dyn_jacket_down_front_01_r
  bone2Name=dyn_jacket_down_front_base_r
  swing1LimitDegrees=15
  swing2LimitDegrees=45
  twistLimitDegrees=15
 }
 {
  bone1Name=dyn_jacket_down_front_01_l
  bone2Name=dyn_jacket_down_front_base_l
  swing1LimitDegrees=15
  swing2LimitDegrees=45
  twistLimitDegrees=15
 }
 {
  bone1Name=dyn_jacket_up_front_base_l
  bone2Name=spine_03
  swing1LimitDegrees=45
  swing1Motion=Free
  swing2LimitDegrees=45
  swing2Motion=Free
  twistLimitDegrees=45
  twistMotion=Free
 }
 {
  bone1Name=dyn_jacket_up_front_base_r
  bone2Name=spine_03
  swing1LimitDegrees=45
  swing1Motion=Free
  swing2LimitDegrees=45
  swing2Motion=Free
  twistLimitDegrees=45
  twistMotion=Free
 }
 {
  bone1Name=dyn_jacket_down_front_base_r
  bone2Name=spine_03
  swing1LimitDegrees=45
  swing1Motion=Free
  swing2LimitDegrees=45
  swing2Motion=Free
  twistLimitDegrees=45
  twistMotion=Free
 }
 {
  bone1Name=dyn_jacket_down_front_base_l
  bone2Name=spine_03
  swing1LimitDegrees=45
  swing1Motion=Free
  swing2LimitDegrees=45
  swing2Motion=Free
  twistLimitDegrees=45
  twistMotion=Free
 }
 {
  bone1Name=dyn_jacket_down_side_01_l
  bone2Name=spine_03
  swing1LimitDegrees=45
  swing1Motion=Free
  swing2LimitDegrees=45
  swing2Motion=Free
  twistLimitDegrees=45
  twistMotion=Free
 }
 {bone1Name=dyn_jacket_down_side_base_l, bone2Name=dyn_jacket_down_side_01_l, swing1LimitDegrees=5, swing2LimitDegrees=45, twistLimitDegrees=5}
 {
  bone1Name=dyn_jacket_down_side_01_r
  bone2Name=spine_03
  swing1LimitDegrees=45
  swing1Motion=Free
  swing2LimitDegrees=45
  swing2Motion=Free
  twistLimitDegrees=45
  twistMotion=Free
 }
 {bone1Name=dyn_jacket_down_side_base_r, bone2Name=dyn_jacket_down_side_01_r, swing1LimitDegrees=5, swing2LimitDegrees=45, twistLimitDegrees=5}
 {
  bone1Name=dyn_jacket_down_back_base
  bone2Name=spine_03
  swing1LimitDegrees=45
  swing1Motion=Free
  swing2LimitDegrees=45
  swing2Motion=Free
  twistLimitDegrees=45
  twistMotion=Free
 }
 {bone1Name=dyn_jacket_down_back_01, bone2Name=dyn_jacket_down_back_base, swing1LimitDegrees=3, swing2LimitDegrees=45, twistLimitDegrees=3}
 {
  bone1Name=spine_03
  bone2Name=pelvis
  swing1LimitDegrees=45
  swing1Motion=Free
  swing2LimitDegrees=45
  swing2Motion=Free
  twistLimitDegrees=45
  twistMotion=Free
 }
 {
  bone1Name=thigh_l
  bone2Name=pelvis
  swing1LimitDegrees=45
  swing1Motion=Free
  swing2LimitDegrees=45
  swing2Motion=Free
  twistLimitDegrees=45
  twistMotion=Free
 }
 {
  bone1Name=thigh_r
  bone2Name=pelvis
  swing1LimitDegrees=45
  swing1Motion=Free
  swing2LimitDegrees=45
  swing2Motion=Free
  twistLimitDegrees=45
  twistMotion=Free
 }
 {
  bone1Name=upperarm_r
  bone2Name=spine_03
  swing1LimitDegrees=45
  swing1Motion=Free
  swing2LimitDegrees=45
  swing2Motion=Free
  twistLimitDegrees=45
  twistMotion=Free
 }
 {
  bone1Name=dyn_jacket_mid_front_base_r
  bone2Name=spine_03
  swing1LimitDegrees=45
  swing1Motion=Free
  swing2LimitDegrees=45
  swing2Motion=Free
  twistLimitDegrees=45
  twistMotion=Free
 }
 {bone1Name=dyn_jacket_mid_front_01_r, bone2Name=dyn_jacket_mid_front_base_r, swing1LimitDegrees=2, swing2LimitDegrees=7, twistLimitDegrees=35}
 {bone1Name=dyn_jacket_mid_front_01_l, bone2Name=dyn_jacket_mid_front_base_l, swing1LimitDegrees=2, swing2LimitDegrees=7, twistLimitDegrees=35}
 {
  bone1Name=dyn_jacket_mid_front_base_l
  bone2Name=spine_03
  swing1LimitDegrees=45
  swing1Motion=Free
  swing2LimitDegrees=45
  swing2Motion=Free
  twistLimitDegrees=45
  twistMotion=Free
 }
 {
  bone1Name=upperarm_l
  bone2Name=spine_03
  swing1LimitDegrees=45
  swing1Motion=Free
  swing2LimitDegrees=45
  swing2Motion=Free
  twistLimitDegrees=45
  twistMotion=Free
 }
 {bone1Name=dyn_jacket_up_front_01_r, bone2Name=dyn_jacket_mid_front_01_r, swing1LimitDegrees=8, swing2LimitDegrees=8, twistLimitDegrees=8}
 {bone1Name=dyn_jacket_up_front_01_l, bone2Name=dyn_jacket_mid_front_01_l, swing1LimitDegrees=8, swing2LimitDegrees=8, twistLimitDegrees=8}
]
